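; Looping gameplay cue, going by its name and folder, that changes the audio mix while Specter mode is in effect.
; BeginPlay activates the CBM_InSpecterMode control bus mix through Audio Modulation, and EndPlay deactivates it.
; The mix follows the actor's lifetime rather than cue-specific events, and the construction script is empty.
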(event EventBeginPlay
  (Audio|Modulation|ActivateControlBusMix 0 "/SpecterGameplay/Audio/Data/CBM_InSpecterMode.CBM_InSpecterMode"))

(event EventEndPlay
  (params (EndPlayReason :type EEndPlayReason%20Enum :default "Destroyed"))
  (Audio|Modulation|DeactivateControlBusMix 0 "/SpecterGameplay/Audio/Data/CBM_InSpecterMode.CBM_InSpecterMode"))

(fn ConstructionScript)
