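; Looping gameplay cue for a Specter pawn that varies its movement audio with how fast it moves.
; On loop start it stores the target pawn and the spawned audio components.
; Each tick it interpolates toward the pawn's squared velocity, then feeds a clamped remap of that value into the first audio component's "Speed" parameter.
; Using squared length avoids a square root but makes the response quadratic in speed.
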(event EventOnLoopingStart(EitherVFX-preferthisversion)
  (params (MyTarget :type Actor%20Object%20Reference) (Parameters :type Gameplay%20Cue%20Parameters%20Structure%20%28by%20ref%29) (ParticleComponents :type Array%20of%20FXSystem%20Component%20Object%20References) (AudioComponents :type Array%20of%20Audio%20Component%20Object%20References))
  (bind _asspecter_pawn (Utilities|Casting|CastToSpecterPawn MyTarget))
  (Variables|Default|SetTargetPlayer _asspecter_pawn)
  (Variables|Default|SetSpawnedAudioComps AudioComponents))

(event EventTick
  (params (DeltaSeconds :type Float%20%28single-precision%29))
  (bind _spawnedaudiocomps (Variables|Default|GetSpawnedAudioComps))
  (bind _output (Utilities|Array|Get(acopy) _spawnedaudiocomps))
  (if (Utilities|Array|IsValidIndex (Variables|Default|GetSpawnedAudioComps))
    (bind _targetplayer (Variables|Default|GetTargetPlayer)
      (:then
        (Variables|Default|SetVelocitySqSmoothed (Math|Interpolation|FInterpTo (Variables|Default|GetVelocitySqSmoothed) (Math|Vector|VectorLengthSquared (Transformation|GetVelocity _targetplayer)) (Utilities|Time|GetWorldDeltaSeconds _targetplayer) 5.0))
        (Utilities|IsValid _output
          (:Is%20Valid
            (Audio|Parameter|SetFloatParameter _output "Speed" (Math|Float|MapRangeClamped (Variables|Default|GetVelocitySqSmoothed) 0.0 2.25e+06 0.0 0.8) :!class /Script/AudioExtensions.AudioParameterControllerInterface))))
      (:else))))

(fn ConstructionScript)
