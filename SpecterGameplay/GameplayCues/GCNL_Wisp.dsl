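; Looping gameplay cue that shows a Specter player's haunt count as Niagara wisps.
; On looping start it caches the target pawn and Niagara component, then listens on the global GameplayEventRouter for haunt-count changes.
; Events for that pawn set the "NumCharges" Niagara variable, capped at a maximum, and EndPlay stops the listener.
; A debug flag replaces the listener with test code that prints and steps the wisp count.
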(event EventOnLoopingStart(EitherVFX-preferthisversion)
  (params (MyTarget :type Actor%20Object%20Reference) (Parameters :type Gameplay%20Cue%20Parameters%20Structure%20%28by%20ref%29) (ParticleComponents :type Array%20of%20FXSystem%20Component%20Object%20References) (AudioComponents :type Array%20of%20Audio%20Component%20Object%20References))
  ;[0.0361 1 0] On Looping Start
  (|CallParentFunction MyTarget Parameters ParticleComponents AudioComponents)
  (bind _asfort_player_pawn (Utilities|Casting|CastToFortPlayerPawn MyTarget))
  (Variables|Default|SetPlayer%20Pawn _asfort_player_pawn)
  (bind _asniagara_particle_system_component (Utilities|Casting|CastToNiagaraComponent (Utilities|Array|Get(acopy) ParticleComponents)))
  (Variables|Default|SetNiagaraSystemComponentCached _asniagara_particle_system_component)
  ; DEBUG
  (if (Variables|Default|GetDebugTestNiagaraVariables)
    ; Debug Test -- Cycle through 0-5 on the number of wisps
    (Niagara|SetNiagaraVariable(Int32) (Variables|Default|GetNiagaraSystemComponentCached) "NumCharges" (Variables|Default|GetNumHauntWisps))
    (Development|PrintString 0 (Utilities|String|Append "NumCharges =" (Utilities|String|ToString(Integer) (Variables|Default|GetNumHauntWisps))))
    (Utilities|FlowControl|Delay 0 2.0)
    (if (Variables|Default|GetisIncrementing)
      (Math|Integer|IncrementInt (Variables|Default|GetNumHauntWisps)
        (:%20%20
          (if (< (Variables|Default|GetNumHauntWisps) 5)
            (Variables|Default|SetNumHauntWisps 0)))) ; Increment
    )
    (Math|Integer|DecrementInt (Variables|Default|GetNumHauntWisps)
      (:%20%20
        (if (< (Variables|Default|GetNumHauntWisps) 0)
          (Variables|Default|SetNumHauntWisps 5)))) ; Decrement
    (else
      ; Listen for router events and set Niagara wisp count accordingly.
      (bind (_asynctaskproxy _payload) (GameplayEventRouter|StartListeningToEvent (Variables|Default|GetPlayer%20Pawn) (GameplayEventRouter|GetGlobalEventRouter 0) "/SpecterGameplay/Gameplay/Event_SpecterNumHauntsChanged.Event_SpecterNumHauntsChanged" (Variables|Default|GetPlayer%20Pawn))
        (:then
          (Variables|Default|SetAsync%20Task%20Ref _asynctaskproxy))
        (:OnEventReceived
          (bind (_payload_livingplayerpawn_7_16a8f6a540ae152135141fa9ac612298 _payload_hauntcount_2_debaa74b44657f889f2436afd90cc98b) (Utilities|Struct|BreakEventSpecterNumHauntsChanged _payload))
          (if (== _payload_livingplayerpawn_7_16a8f6a540ae152135141fa9ac612298 (Variables|Default|GetPlayer%20Pawn))
            (bind _output_get (Variables|Default|SetNumHauntWisps (Math|Integer|Min(Integer) _payload_hauntcount_2_debaa74b44657f889f2436afd90cc98b (Variables|Default|GetMaxHauntsWispsInNiagara))))
            (Niagara|SetNiagaraVariable(Int32) (Variables|Default|GetNiagaraSystemComponentCached) "NumCharges" _output_get)))))))

(event EventEndPlay
  (params (EndPlayReason :type EEndPlayReason%20Enum :default "Destroyed"))
  (bind _async_task_ref (Variables|Default|GetAsync%20Task%20Ref))
  (|CallParentFunction EndPlayReason)
  (Utilities|IsValid _async_task_ref
    (:Is%20Valid
      (GameplayEventRouter|StopListeningtoEvent _async_task_ref))))

(fn ConstructionScript)
